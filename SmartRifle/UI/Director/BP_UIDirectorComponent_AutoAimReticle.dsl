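; UI director component that shows or hides the SmartRifle auto-aim lock-on reticle scene for the owning local player.
; When ready to show UI, it listens on the global GameplayEventRouter for lock-on reticle state change events and stores the listener handle.
; Each event's broken-out payload drives adding or removing the smart reticle scene.
; On EndPlay it stops listening and removes the scene.
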(event EventOnReadyToShowUI
  (bind (_asynctaskproxy _payload) (GameplayEventRouter|StartListeningToEvent (GameplayEventRouter|GetGlobalEventRouter) "/Script/AutoAimWeaponRuntime.FortEvent_LockOnReticleStateChanged" (CallFunction|GetOwningLocalPlayerController))
    (:then
      (Variables|Default|SetOnEnabledSmartReticle _asynctaskproxy))
    (:OnEventReceived
      (CallFunction|AddorRemoveScene (Variables|Default|GetSmartReticleEnabled) (Utilities|Struct|BreakFortEventLockonReticleStateChanged _payload)))))

(event EventEndPlay
  (params (EndPlayReason :type EEndPlayReason%20Enum :default "Destroyed"))
  (GameplayEventRouter|StopListeningtoEvent (Variables|Default|GetOnEnabledSmartReticle))
  (CallFunction|RemoveScene (Variables|Default|GetSmartReticleEnabled)))
